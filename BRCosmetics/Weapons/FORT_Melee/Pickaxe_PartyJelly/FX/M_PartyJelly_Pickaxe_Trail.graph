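# Unlit translucent Niagara ribbon material for the PartyJelly pickaxe trail, tinting pale blue to white by particle color.
# Two opposite-panning light-ray texture samples blend into streak noise that fades along the trail.
# High-power sines create four procedural lines, and a UV-based edge fade takes its leading-edge hardness from a dynamic parameter.
# Named reroutes feed these masks into color and opacity, with opacity contrast-shaped and scaled by particle alpha.
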Material /BRCosmetics/Weapons/FORT_Melee/Pickaxe_PartyJelly/FX/M_PartyJelly_Pickaxe_Trail
graph {
 properties {
  bAutomaticallySetUsageInEditor=false
  bUsedWithNiagaraRibbons=true
  blendMode=BLEND_Translucent
  shadingModel=MSM_Unlit
  usedShadingModels=Unlit
 }
 nodes [
  n1 = Constant3Vector() {constant=<rgba 0.64 0.64 1 1>}
  n2 = NamedRerouteUsage()
  n3 = ParticleColor()
  MaterialFunctionCall(Emissive=Multiply(A=LinearInterpolate(A=@n1, Alpha=@n2) {constA=1}, B=@n3.RGB)) {materialFunction=&MF_FX_EmissiveAdjustments}
  n4 = NamedRerouteUsage()
  n5 = OneMinus(None=MaterialFunctionCall(Float2=TextureCoordinate()) {materialFunction=&BreakOutFloat2Components}.R)
  n6 = MaterialFunctionCall(UVs=TextureCoordinate(), Rotation Center=Constant(), "Rotation Angle (0-1)"=Constant() {r=0.25}) {materialFunction=&CustomRotator}
  n7 = Constant2Vector() {g=1, r=2.3}
  n8 = TextureSampleParameter2D(UVs=Panner(Coordinate=Multiply(A=@n6.Rotated Values, B=@n7)) {speedX=-0.7}) {parameterName=Tex1, texture=/Game/Effects/Fort_Effects/Textures/Outlander/T_InTheZone_LightRays_01}
  n9 = LinearInterpolate(
   A=TextureSampleParameter2D(UVs=Panner(Coordinate=@n6.Rotated Values) {speedX=0.5}) {parameterName=Tex1, texture=&T_InTheZone_LightRays_01}.R
   B=Multiply(A=Power(Base=@n8.G) {constExponent=3}) {constB=1.4}
  )
  n10 = LinearInterpolate(A=Multiply(A=Power(Base=@n5) {constExponent=1}, B=LinearInterpolate(Alpha=@n9)), Alpha=Power(Base=@n5) {constExponent=4})
  n11 = Multiply(A=NamedRerouteUsage(), B=Saturate(None=Add(A=@n4, B=@n10)))
  n12 = Multiply(A=@n11) {constB=0.45}
  n13 = Saturate(None=Multiply(A=Power(Base=@n11) {constExponent=4}) {constB=3})
  Multiply(A=@n3.A, B=Saturate(None=Add(A=@n12, B=@n13)))
  n14 = LinearInterpolate(Alpha=DynamicParameter() {paramNames [LeadingEdgeHardness, Param2, Param3, Param4]}.LeadingEdgeHardness) {constA=1, constB=15}
  n15 = MaterialFunctionCall(Float2=TextureCoordinate()) {materialFunction=&BreakOutFloat2Components}
  n16 = Sine(None=Multiply(A=@n15.G) {constB=0.5})
  n17 = Saturate(None=Multiply(A=@n16, B=LinearInterpolate(Alpha=@n15.G) {constA=2}) {constB=1.5})
  n18 = Saturate(None=Multiply(A=LinearInterpolate(B=@n14, Alpha=@n17) {constA=2, constB=15}, B=@n15.R) {constB=18})
  n19 = OneMinus(None=@n15.R)
  n20 = Multiply(A=@n16) {constB=4}
  NamedRerouteDeclaration(None=Multiply(A=Multiply(A=@n18, B=@n19) {constA=1}, B=Saturate(None=@n20))) {name=Edge Fade, nodeColor=<rgba 0.0781 0 1 1>}
  n21 = Multiply(A=Add(A=@n15.G) {constB=0.45}) {constB=0.5}
  n22 = LinearInterpolate(Alpha=Power(Base=@n15.R) {constExponent=1.5}) {constA=100, constB=200}
  n23 = Multiply(A=Add(A=@n15.G) {constB=0.17}) {constB=0.5}
  n24 = Add(A=Power(Base=Sine(None=@n21), Exp=@n22) {constExponent=500}, B=Power(Base=Sine(None=@n23), Exp=@n22) {constExponent=500})
  n25 = Add(A=@n15.G) {constB=-0.1}
  n26 = Power(Base=Sine(None=Multiply(A=@n25) {constB=0.5}), Exp=@n22) {constExponent=500}
  n27 = Sine(None=Multiply(A=Add(A=@n15.G) {constB=-0.45}) {constB=0.5})
  NamedRerouteDeclaration(None=Add(A=Add(A=@n24, B=@n26), B=Power(Base=@n27, Exp=@n22) {constExponent=500})) {name=Lines, nodeColor=<rgba 1 0.2907 0.9008 1>}
 ]
}
